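; NPC pawn for the insect swarm wildlife AI, which can settle on plants, turn hostile, or be gathered for an item when calm.
; Server listeners for gameplay tags and events set replicated state; OnRep handlers update leash, interaction, plant collision, Niagara and sound.
; Damage or launch forces it off plants.
; Clients trace for dust and manage a distance-culled wash force actor; death effects differ when gathered.
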(event EventBeginPlay
  (bind _tagappliedwhenhostile (Variables|Default|GetTagAppliedWhenHostile))
  (bind _tagappliedwhenonplant (Variables|Default|GetTagAppliedWhenOnPlant))
  (|CallParentFunction)
  (CallFunction|UpdateSoundLibraryAudio "(TagName=\"SoundLibrary.InsectSwarm.Loop\")" true)
  (if (ScalableFloat|GetValueasBool (Variables|Default|GetSF_HideHealthbar))
    (CallFunction|NPCDisableHealthBars)
    (else
      _))
  (Networking|SwitchHasAuthority
    (:Authority
      (CallFunction|InitPawn)
      (Utilities|FlowControl|Sequence
        (:then_0
          ;[1 1 1] Listen for when the AI is supposed to be on a Plant
          (bind (_asynctaskproxy _payload) (Ability|Async|WaitGameplayEventToActor 0 (Variables|Default|GetTag_FoundPlant))
            (:EventReceived
              (CallFunction|RelocateSwarmtoPlant _payload))))
        (:then_1
          (Ability|Tasks|WaitGameplayTagAddToActor 0 _tagappliedwhenonplant
            (:Added
              (Ability|Tasks|WaitGameplayTagRemoveFromActor 0 _tagappliedwhenonplant
                (:then
                  (CallFunction|SwarmIsonPlant))
                (:Removed
                  (CallFunction|SwarmIsOffPlant))))))
        (:then_2
          ;[1 1 1] Listen for when AI is Hostile or not
          (Ability|Tasks|WaitGameplayTagAddToActor 0 _tagappliedwhenhostile
            (:Added
              (Ability|Tasks|WaitGameplayTagRemoveFromActor 0 _tagappliedwhenhostile
                (:then
                  (Variables|InternalData|SetbSwarmIsHostile true))
                (:Removed
                  (Variables|InternalData|SetbSwarmIsHostile false))))))
        (:then_3
          ;[1 1 1] Listen for when Swarm is trying to break down walls
          (Ability|Async|WaitGameplayEventToActor 0 (Variables|Default|GetTagWhenAttackingWhileBlocked)
            (:EventReceived
              (Variables|InternalData|SetbSwarmIsHostile true)
              (Utilities|FlowControl|Delay 1.5)
              (Variables|InternalData|SetbSwarmIsHostile false))))))
    (:Remote
      ;[1 1 1] WashForceAndDustFX
      (Utilities|FlowControl|Sequence
        (:then_0
          (bind _returnvalue (Utilities|Time|SetTimerbyEvent (delegate Custom|UpdateDustLocation) 0.3 true))
          (Variables|Default|SetUpdateDustLocationTimer _returnvalue))
        (:then_1
          (Utilities|EffectsQualityCheck 0
            (:Cinematic
              _)
            (:Epic
              _))
          (Variables|Default|GetPhysicalForceActor
            (:then)
            (:else
              (Utilities|FlowControl|DelayUntilNextTick)
              (if (not (FortPawn|IsDead))
                (bind _returnvalue_2 (Game|SpawnActorfromClass "/Game/Effects/Niagara/Blueprints/PhsyicsFields/FN_WashForce.FN_WashForce_C" (Transformation|GetActorTransform) "Undefined" "MultiplyWithRoot" 0 7e+02 0.0 -999.0 0))
                (Variables|Default|SetPhysicalForceActor _returnvalue_2)
                (Transformation|AttachActorToActor (Variables|Default|GetPhysicalForceActor) self "None" "SnapToTarget" "KeepRelative" "KeepRelative" true)
                (Development|PrintString "WashForceActorSpawned" true true "(R=1.000000,G=0.107830,B=0.000000,A=1.000000)" 1e+01)))))))))

(event Custom|UpdateDustLocation
  (bind _timeline_forcestrength (Variables|NPC_Pawn_Swarm|GetTimeline_ForceStrength))
  ;[1 1 1] WashForceAndDustFX
  (if (Variables|InternalData|GetIsOnPlant)
    (Utilities|FlowControl|DoOnce false
      (:Completed
        (Components|Timeline|PlayfromStart _timeline_forcestrength)))
    (else
      (bind _outhit (Collision|LineTraceByChannel :Start (+ (Transformation|GetActorLocation) "0.000000,0.000000,100.000000") :End (+ (Transformation|GetActorLocation) "0.000000,0.000000,-1000.000000") :TraceChannel "TraceTypeQuery1" :bTraceComplex false :DrawDebugType "None" :bIgnoreSelf true :TraceColor "(R=1.000000,G=0.000000,B=0.000000,A=1.000000)" :TraceHitColor "(R=0.000000,G=1.000000,B=0.000000,A=1.000000)" :DrawTime 2.0))
      (Niagara|SetNiagaraVariable(Position) (Variables|Default|GetNS_AI_Wildlife_InsectSwarm) "DustLocation" (.Location (Collision|BreakHitResult _outhit)))
      (Utilities|FlowControl|Sequence
        (:then_0
          _)
        (:then_1
          (for _array_element (Fortnite|GetLocalFortPlayerControllers true)
            (if (> (Math|Vector|Distance(Vector) (Transformation|GetActorLocation (Game|Player|GetPlayerPawn _array_element)) (.Location (Collision|BreakHitResult _outhit))) 4e+03)
              (CallFunction|CEDestroyWashForceActor)
              (else
                (CallFunction|CESpawnWashForceActor)
                (bind _physicalforceactor (Variables|Default|GetPhysicalForceActor)
                  (:then
                    (Variables|/Game/Effects/Niagara/Blueprints/PhsyicsFields/FN_WashForce.FN_WashForce_C|SetStrength 1.1 _physicalforceactor))
                  (:else)))))))))
  (Utilities|FlowControl|DoOnce false
    (:Completed
      (CallFunction|CESpawnWashForceActor)
      (bind _physicalforceactor_2 (Variables|Default|GetPhysicalForceActor)
        (:then
          (Variables|/Game/Effects/Niagara/Blueprints/PhsyicsFields/FN_WashForce.FN_WashForce_C|SetStrength 1.1 _physicalforceactor_2))
        (:else
          _))
      (Components|Timeline|ReversefromEnd _timeline_forcestrength))))

(event Custom|CE_SpawnWashForceActor
  ;[1 1 1] WashForceAndDustFX
  (Utilities|EffectsQualityCheck 0
    (:Cinematic
      _)
    (:Epic
      _))
  (Variables|Default|GetPhysicalForceActor
    (:then)
    (:else
      (Utilities|FlowControl|DelayUntilNextTick)
      (if (not (FortPawn|IsDead))
        (bind _returnvalue (Game|SpawnActorfromClass "/Game/Effects/Niagara/Blueprints/PhsyicsFields/FN_WashForce.FN_WashForce_C" (Transformation|GetActorTransform) "Undefined" "MultiplyWithRoot" 0 7e+02 0.0 -999.0 0))
        (Variables|Default|SetPhysicalForceActor _returnvalue)
        (Transformation|AttachActorToActor (Variables|Default|GetPhysicalForceActor) self "None" "SnapToTarget" "KeepRelative" "KeepRelative" true)
        (Development|PrintString "WashForceActorSpawned" true true "(R=1.000000,G=0.107830,B=0.000000,A=1.000000)" 1e+01)))))

(event Custom|CE_DestroyWashForceActor
  ;[1 1 1] WashForceAndDustFX
  (bind _physicalforceactor (Variables|Default|GetPhysicalForceActor)
    (:then
      (Actor|DestroyActor _physicalforceactor)
      (Development|PrintString "WashForceActorDestroyed" true true "(R=0.095265,G=1.000000,B=0.012008,A=1.000000)" 1e+01))
    (:else)))

(event Custom|RelocateSwarmToPlant
  (params (EventData :type Gameplay%20Event%20Data%20Structure))
  (bind (_eventtag _instigator _target _optionalobject) (Utilities|Struct|BreakGameplayEventData EventData))
  ;[1 1 1] WashForceAndDustFX
  ;[1 1 1] Reposition Visual of Actor on Plant
  (Utilities|IsValid _optionalobject
    (:Is%20Valid
      (bind _asactor (Utilities|Casting|CastToActor _optionalobject))
      (Variables|InternalData|SetCurrentPlantActor _asactor) ; The logic needs to also run on the client
      (bind _asbuilding_gameplay_actor (Utilities|Casting|CastToBuildingGameplayActor _asactor))
      (EventDispatchers|BindEventtoOnDamaged _asbuilding_gameplay_actor (EventDispatchers|CreateEvent :!event Custom|PlantWasDamaged)))))

(event Custom|PlantWasDamaged
  (params (DamagedActor :type Actor%20Object%20Reference) (Damage :type Float%20%28single-precision%29) (InstigatedBy :type Controller%20Object%20Reference) (DamageCauser :type Actor%20Object%20Reference) (HitLocation :type Vector) (FHitComponent :type Primitive%20Component%20Object%20Reference) (BoneName :type Name) (Momentum :type Vector))
  ;[1 1 1] Reposition Visual of Actor on Plant
  (CallFunction|ForceRemoveonPlantGE))

(event EventEndPlay
  (params (EndPlayReason :type EEndPlayReason%20Enum :default "Destroyed"))
  (bind _currentplantactor (Variables|InternalData|GetCurrentPlantActor))
  (Utilities|IsValid _currentplantactor
    (:Is%20Valid
      (CallFunction|RemoveIgnoredActoronPlantEnd _currentplantactor EndPlayReason))))

(event EventOnDeathServer
  (params (Damage :type Float%20%28single-precision%29) (DamageTags :type Gameplay%20Tag%20Container%20Structure%20%28by%20ref%29) (Momentum :type Vector :default "0, 0, 0") (HitInfo :type Hit%20Result%20Structure%20%28by%20ref%29) (InstigatedBy :type Controller%20Object%20Reference) (DamageCauser :type Actor%20Object%20Reference) (EffectContext :type Gameplay%20Effect%20Context%20Handle%20Structure))
  (|CallParentFunction Damage DamageTags Momentum HitInfo InstigatedBy DamageCauser EffectContext)
  (CallFunction|CEDestroy)
  (if (GameplayTags|HasTag DamageTags "(TagName=\"Gameplay.Damage.OutsideSafeZone\")")
    (Variables|Default|SetDeathByOutsideSafeZone true)))

(event Custom|DisableCollisionSwarmToPlant
  (bind _currentplantactor (Variables|InternalData|GetCurrentPlantActor))
  ;[1 1 1] Reposition Visual of Actor on Plant
  (Variables|InternalData|SetPlantSphereCollisionComponent (Actor|GetComponentbyClass _currentplantactor "/Script/Engine.SphereComponent"))
  (Game|BindEventtoOnEndPlay _currentplantactor (delegate Custom|RemoveIgnoredActorOnPlantEnd))
  (Collision|IgnoreActorwhenMoving (Variables|Character|GetCapsuleComponent) _currentplantactor true) ; Make sure to bin on EndPlay to remove the ignore
  (Collision|IgnoreActorwhenMoving (Variables|InternalData|GetPlantSphereCollisionComponent) self true)
  (Variables|InternalData|SetPlantActorLocation (Transformation|GetActorLocation _currentplantactor))
  (Variables|InternalData|SetSwarmOriginalPositionBeforePlantMove (Transformation|GetActorLocation self))
  (bind (_direction _newtrack_0) (Timeline|Timeline PlayFromStart)
    (:Update
      (Transformation|SetActorLocation self (Math|Vector|Lerp(Vector) (Variables|InternalData|GetSwarmOriginalPositionBeforePlantMove) (Variables|InternalData|GetPlantActorLocation) _newtrack_0) false true))))

(event Custom|CE_Destroy
  (Networking|CanExecuteCosmeticEvents
    (:True
      (CallFunction|CEDestroyWashForceActor))))

(event Custom|RemoveIgnoredActorOnPlantEnd
  (params (Actor :type Actor%20Object%20Reference) (EndPlayReason :type EEndPlayReason%20Enum :default "Destroyed"))
  (bind _currentplantactor (Variables|InternalData|GetCurrentPlantActor))
  ;[1 1 1] Reposition Visual of Actor on Plant
  (Utilities|IsValid _currentplantactor
    (:Is%20Valid
      (Collision|IgnoreActorwhenMoving (Variables|Character|GetCapsuleComponent) _currentplantactor false)
      (Collision|IgnoreActorwhenMoving (Variables|InternalData|GetPlantSphereCollisionComponent) self false)
      (Game|UnbindEventfromOnEndPlay _currentplantactor (EventDispatchers|CreateEvent :!event Custom|RemoveIgnoredActorOnPlantEnd))
      (Networking|SwitchHasAuthority
        (:Authority
          (bind _asbuilding_gameplay_actor (Utilities|Casting|CastToBuildingGameplayActor (Variables|InternalData|GetCurrentPlantActor)))
          (EventDispatchers|UnbindEventfromOnDamaged _asbuilding_gameplay_actor (EventDispatchers|CreateEvent :!event Custom|PlantWasDamaged)))))))

(event Custom|ResetCollisionSwarmToPlant
  (bind _currentplantactor (Variables|InternalData|GetCurrentPlantActor))
  ;[1 1 1] Reposition Visual of Actor on Plant
  (Utilities|FlowControl|Delay 2.0)
  (Utilities|IsValid _currentplantactor
    (:Is%20Valid
      (Collision|IgnoreActorwhenMoving (Variables|Character|GetCapsuleComponent) _currentplantactor false)
      (Collision|IgnoreActorwhenMoving (Variables|InternalData|GetPlantSphereCollisionComponent) self false)
      (Game|UnbindEventfromOnEndPlay _currentplantactor (EventDispatchers|CreateEvent :!event Custom|RemoveIgnoredActorOnPlantEnd))
      (Networking|SwitchHasAuthority
        (:Authority
          (bind _asbuilding_gameplay_actor (Utilities|Casting|CastToBuildingGameplayActor (Variables|InternalData|GetCurrentPlantActor)))
          (EventDispatchers|UnbindEventfromOnDamaged _asbuilding_gameplay_actor (EventDispatchers|CreateEvent :!event Custom|PlantWasDamaged)))))))

(event EventOnDeathPlayEffects
  (params (Damage :type Float%20%28single-precision%29) (DamageTags :type Gameplay%20Tag%20Container%20Structure%20%28by%20ref%29) (Momentum :type Vector :default "0, 0, 0") (HitInfo :type Hit%20Result%20Structure%20%28by%20ref%29) (InstigatedBy :type Fort%20Pawn%20Object%20Reference) (DamageCauser :type Actor%20Object%20Reference) (EffectContext :type Gameplay%20Effect%20Context%20Handle%20Structure))
  (bind _interactingpawn (Variables|InternalData|GetInteractingPawn))
  (bind _mesh (Variables|/Script/Engine.Character|GetMesh _interactingpawn))
  (bind _returnvalue (Transformation|GetSocketLocation _mesh "hand_r"))
  (bind _ns_ai_wildlife_insectswarm (Variables|Default|GetNS_AI_Wildlife_InsectSwarm))
  ;[1 1 1] On Death Effects
  (|CallParentFunction Damage DamageTags Momentum HitInfo InstigatedBy DamageCauser EffectContext)
  (Utilities|Time|ClearandInvalidateTimerbyHandle (Variables|Default|GetUpdateDustLocationTimer))
  (CallFunction|UpdateSoundLibraryAudio "(TagName=\"SoundLibrary.InsectSwarm.Loop\")")
  (if (and (Variables|InternalData|GetDeathByGather) (Utilities|IsValid _interactingpawn :!class /Script/Engine.KismetSystemLibrary))
    (Utilities|FlowControl|Delay 0.4)
    (Niagara|SetNiagaraVariable(Float) _ns_ai_wildlife_insectswarm "AttractToHand" 1.0)
    (Niagara|SetNiagaraVariable(Position) _ns_ai_wildlife_insectswarm "CollectHandPosition" _returnvalue)
    (Utilities|FlowControl|Delay 0.1)
    (Niagara|SetNiagaraVariable(Position) _ns_ai_wildlife_insectswarm "CollectHandPosition" _returnvalue)
    (Utilities|FlowControl|Delay 0.1)
    (Niagara|SetNiagaraVariable(Position) _ns_ai_wildlife_insectswarm "CollectHandPosition" _returnvalue)
    (else
      (CallFunction|SpawnDeathEffects)
      (CallFunction|UpdateSoundLibraryAudio "(TagName=\"SoundLibrary.InsectSwarm.Death\")" true))))

(event Custom|SpawnDeathEffects
  (bind _ns_ai_wildlife_insectswarm (Variables|Default|GetNS_AI_Wildlife_InsectSwarm))
  ;[1 1 1] On Death Effects
  (if (not (ScalableFloat|GetValueasBool (Variables|Default|GetSF_SpawnBGAOnDeath)))
    (Effects|Components|ParticleSystem|SpawnFXAtLocation "/ChronoSwarmGrenade/FX/NS_AI_Wildlife_InsectSwarm_Death.NS_AI_Wildlife_InsectSwarm_Death" (+ (Transformation|GetActorLocation) "0.000000,0.000000,88.000000")))
  (Effects|Components|ParticleSystem|SetFloatParameter _ns_ai_wildlife_insectswarm "Hostile" 0.0)
  (Components|Activation|Deactivate _ns_ai_wildlife_insectswarm)
  (CallFunction|NPCDisableHealthBars))

(event EventOnDamagePlayEffects
  (params (Damage :type Float%20%28single-precision%29) (DamageTags :type Gameplay%20Tag%20Container%20Structure%20%28by%20ref%29) (Momentum :type Vector :default "0, 0, 0") (HitInfo :type Hit%20Result%20Structure%20%28by%20ref%29) (InstigatedBy :type Fort%20Pawn%20Object%20Reference) (DamageCauser :type Actor%20Object%20Reference) (EffectContext :type Gameplay%20Effect%20Context%20Handle%20Structure))
  (bind (_bblockinghit _binitialoverlap _time _distance _location) (Collision|BreakHitResult HitInfo))
  (bind _ns_ai_wildlife_insectswarm (Variables|Default|GetNS_AI_Wildlife_InsectSwarm))
  ;[1 1 1] On Damaged
  (|CallParentFunction Damage DamageTags Momentum HitInfo InstigatedBy DamageCauser EffectContext)
  (bind _hitfx (Variables|Default|GetHitFX)
    (:then
      (Components|Activation|Activate _hitfx true)
      (Transformation|SetWorldLocation _hitfx _location false false)
      (Variables|Default|SetHitFX (stale _returnvalue)))
    (:else
      (bind _returnvalue (Niagara|SpawnSystematLocation "/InsectSwarm/FX/NS_AI_Wildlife_InsectSwarm_ByHit.NS_AI_Wildlife_InsectSwarm_ByHit" _location))))
  (Effects|Components|ParticleSystem|SetBoolParameter _ns_ai_wildlife_insectswarm "ByShoot" true)
  (Niagara|SetNiagaraVariable(Position) _ns_ai_wildlife_insectswarm "ShootPosition" _location)
  (Utilities|FlowControl|Delay 0.5)
  (Effects|Components|ParticleSystem|SetBoolParameter _ns_ai_wildlife_insectswarm "ByShoot" false))

(event Custom|HandleLeashUpdateWhenOnPlant
  (Variables|InternalData|SetLeashLocation (Transformation|GetActorLocation self)))

(event Custom|HandleLeashPositionUpdate
  (bind _self self)
  (bind _returnvalue (Pawn|GetController _self))
  (bind _returnvalue_2 (AI|GetBlackboard _returnvalue))
  (bind _goalactorbbname (Variables|InternalData|GetGoalActorBBNAme))
  (bind _returnvalue_3 (AI|Components|Blackboard|GetValueasObject _returnvalue_2 _goalactorbbname))
  ;[1 1 1] Update Leash Position when Swarm gets a new Target so we reduce chances that a Swarm can be at the edge of his leash and then chaces a player for just a second of two.
  (Utilities|IsValid _returnvalue
    (:Is%20Valid
      (Utilities|IsValid _returnvalue_2
        (:Is%20Valid
          (Utilities|IsValid _returnvalue_3
            (:Is%20Valid
              (if (not (== (Variables|InternalData|GetTargetActor) _returnvalue_3))
                (bind _asactor (Utilities|Casting|CastToActor _returnvalue_3))
                (Variables|InternalData|SetLeashLocation (Transformation|GetActorLocation self))
                (Variables|InternalData|SetTargetActor _asactor)))))))))

(event Custom|SwarmIsOnPlant
  (Variables|InternalData|SetIsOnPlant true))

(event Custom|SwarmIsOffPlant
  (bind _returnvalue (Ability|GetAbilitySystemComponent (Variables|InternalData|GetCurrentPlantActor)))
  (bind _currentplantactor (Variables|InternalData|GetCurrentPlantActor))
  (bind _tag_plantinuse (Variables|Default|GetTag_PlantInUse))
  ;[1 1 1] Swarm is off Plant, if it was reserved, release it so another swarm can try to take it.
  (Variables|InternalData|SetIsOnPlant false)
  (Utilities|IsValid _currentplantactor
    (:Is%20Valid
      (bind _returnvalue_2 (GameplayTags|HasMatchingGameplayTag(Message) _currentplantactor _tag_plantinuse))
      (if _returnvalue_2
        (Utilities|IsValid _returnvalue
          (:Is%20Valid
            (GameplayEffects|RemoveActiveEffectswithTags _returnvalue (GameplayTags|MakeGameplayTagContainerfromTag _tag_plantinuse))))
        (else
          _))
      (Variables|InternalData|SetCurrentPlantActor 0))))

(event Custom|HandleSwarmInteract
  ;[1 1 1] Handle Interaction Enabled / Disabled based with Hostility
  (if (Variables|InternalData|GetbSwarmIsHostile)
    (AI|SetCanInteract false)
    (Utilities|Time|ClearandInvalidateTimerbyHandle (Variables|InternalData|GetInteractionEnableTimerHandle))
    (else
      (bind _returnvalue (Utilities|Time|SetTimerbyEvent (delegate Custom|EnableInteraction) 2.5))
      (Variables|InternalData|SetInteractionEnableTimerHandle _returnvalue))))

(event EventOnNPCPawnInteractedWith
  (params (AIPawn :type Actor%20Object%20Reference) (InteractingPlayerPawn :type Fort%20Player%20Pawn%20Object%20Reference) (InteractionBeingAttempted :type EInteractionBeingAttempted%20Enum :default "FirstInteraction"))
  ;[1 1 1] WHen Interacting with the Plant (Gaterhing the Item)
  (Variables|InternalData|SetInteractingPawn InteractingPlayerPawn)
  (bind _asfort_player_controller (Utilities|Casting|CastToFortPlayerController (Pawn|GetController InteractingPlayerPawn)))
  (Utilities|FlowControl|DoOnce false
    (:Completed
      (Variables|InternalData|SetDeathByGather true)
      (Ability|GameplayEffect|AddLooseGameplayTags self (Variables|Default|GetTagsToApplyOnGather))
      (Utilities|FlowControl|Delay 0.1)
      (Inventory|GiveItemtoInventoryOwner :InventoryOwner _asfort_player_controller :ItemDefinition (Variables|Default|GetItemDefToGiveOnGather) :NumberToGive (Math|Float|Truncate (ScalableFloat|GetValueatLevel (Variables|Default|GetSwarmGrenadeCountToGive))) :bNotifyPlayer true :ItemLevel -1 :PickupInstigatorHandle -1 :bUseItemPickupAnalyticEvent false :WeaponAmmoOverride -1)
      (CallFunction|NPCDisableHealthBars)
      (FortPawn|ForceKill "(TagName=\"Pawn.Athena.NPC.Wildlife.InsectSwarm.DeathByGather\")"))))

(event Custom|EnableInteraction
  ;[1 1 1] Handle Interaction Enabled / Disabled based with Hostility
  (AI|SetCanInteract true))

(event EventOnLaunched
  (params (LaunchVelocity :type Vector :default "0, 0, 0") (bXYOverride :type Boolean) (bZOverride :type Boolean))
  (|CallParentFunction LaunchVelocity bXYOverride bZOverride)
  (CallFunction|ForceRemoveonPlantGE))

(event EventOnDamageServer
  (params (Damage :type Float%20%28single-precision%29) (DamageTags :type Gameplay%20Tag%20Container%20Structure%20%28by%20ref%29) (Momentum :type Vector :default "0, 0, 0") (HitInfo :type Hit%20Result%20Structure%20%28by%20ref%29) (InstigatedBy :type Controller%20Object%20Reference) (DamageCauser :type Actor%20Object%20Reference) (EffectContext :type Gameplay%20Effect%20Context%20Handle%20Structure))
  ;[1 1 1] When Damaged, Remove the OnPlant GE to force Swarm Off Plant
  (|CallParentFunction Damage DamageTags Momentum HitInfo InstigatedBy DamageCauser EffectContext)
  (CallFunction|ForceRemoveonPlantGE))

(event Custom|ForceRemoveOnPlantGE
  (bind _tagappliedwhenonplant (Variables|Default|GetTagAppliedWhenOnPlant))
  ;[1 1 1] Force Remove OnPlantGE which will make Swarm try to go back to do normal behaviors
  (if (GameplayTags|HasMatchingGameplayTag self _tagappliedwhenonplant)
    (GameplayEffects|RemoveActiveEffectswithTags (Variables|GameplayAbilities|GetAIPawnAbilitySystemComponent) (GameplayTags|MakeGameplayTagContainerfromTag _tagappliedwhenonplant))))

(event Custom|InsectSwarm_DisplayLeash
  (params (Duration :type Float%20%28double-precision%29))
  (bind _randomdebugcolorselected (Variables|Debug|GetRandomDebugColorSelected))
  (bind _radiusdebugcolorarray (Variables|Debug|GetRadiusDebugColorArray))
  (bind _output (Utilities|Array|Get(aref) _radiusdebugcolorarray _randomdebugcolorselected))
  (bind _leashlocation (Variables|InternalData|GetLeashLocation))
  ;[1 1 1] Draw Debug Data for Swarm
  (Variables|Debug|SetRandomDebugColorSelected (Math|Random|RandomIntegerinRange 0 (- (Utilities|Array|Length (Variables|Debug|GetRadiusDebugColorArray)) 1)))
  (Rendering|Debug|DrawDebugSphere _leashlocation (ScalableFloat|GetValueatLevel (Variables|Default|GetLeashRadius)) 12 _output Duration)
  (Rendering|Debug|DrawDebugSphere _leashlocation 4e+01 12 _output Duration)
  (Rendering|Debug|DrawDebugString (+ _leashlocation "0.000000,0.000000,45.000000") (Utilities|String|Append "Center of leash for: " (Utilities|GetDisplayName self)) 0 _output Duration))

(fn BlueprintCanInteract
  (params (InteractingPawn :type Fort%20Pawn%20Object%20Reference) (InteractionBeingAttempted :type EInteractionBeingAttempted%20Enum :default "FirstInteraction"))
  (return (not (Variables|InternalData|GetbSwarmIsHostile))))

(fn InitPawn
  (Variables|InternalData|SetLeashLocation (Transformation|GetActorLocation self))
  (Variables|InternalData|SetDeathByGather false)
  (return))

(fn IsInsideLeashRadius
  (params (ActorToTest :type Actor%20Object%20Reference) (VectorToTest :type Vector))
  (Utilities|IsValid ActorToTest
    (:Is%20Valid
      (LocalVariables|SetLocationToTest (Transformation|GetActorLocation ActorToTest)))
    (:Is%20Not%20Valid
      (LocalVariables|SetLocationToTest VectorToTest)))
  (return (<= (Math|Vector|Distance2D(Vector) (LocalVariables|GetLocationToTest) (Variables|InternalData|GetLeashLocation)) (ScalableFloat|GetValueatLevel (Variables|Default|GetLeashRadius)))))

(fn IsValidAutoFireTarget
  (params (Attacker :type Actor%20Object%20Reference))
  (return (and (Utilities|Enum|Equal(Enum) (Team|GetTeamAffiliation Attacker self) "Hostile") (Variables|InternalData|GetbSwarmIsHostile))))

(fn OnRep_bSwarmIsHostile
  (Networking|SwitchHasAuthority
    (:Authority
      (CallFunction|HandleSwarmInteract)
      (CallFunction|HandleLeashPositionUpdate))
    (:Remote
      (Effects|Components|ParticleSystem|SetFloatParameter (Variables|Default|GetNS_AI_Wildlife_InsectSwarm) "Hostile" (select (Variables|InternalData|GetbSwarmIsHostile) 1.0 0.0))
      (Class|InsectSwarmSoundLibraryComponent|UpdateHostileStatus (Variables|Default|GetInsectSwarm_SoundLibraryComponent) (Variables|InternalData|GetbSwarmIsHostile)))))

(fn OnRep_CurrentPlantActor
  (Utilities|IsValid (Variables|InternalData|GetCurrentPlantActor)
    (:Is%20Valid
      (CallFunction|DisableCollisionSwarmtoPlant))))

(fn OnRep_DeathByGather
  (Networking|SwitchHasAuthority
    (:Remote
      (if (Variables|InternalData|GetDeathByGather)
        (Animation|PlayAnimMontage (Variables|InternalData|GetInteractingPawn) (Variables|Default|GetGatherAnimationMontage) 1.0 "None")
        (CallFunction|UpdateSoundLibraryAudio "(TagName=\"SoundLibrary.InsectSwarm.Loop\")")
        (CallFunction|UpdateSoundLibraryAudio "(TagName=\"SoundLibrary.InsectSwarm.Capture\")" true)))))

(fn OnRep_DeathByOutsideSafeZone)

(fn OnRep_InteractingPawn)

(fn OnRep_IsOnPlant
  (Utilities|FlowControl|Sequence
    (:then_0
      (Networking|SwitchHasAuthority
        (:Authority
          (if (Variables|InternalData|GetIsOnPlant)
            (CallFunction|HandleLeashUpdatewhenonPlant)))
        (:Remote
          (Effects|Components|ParticleSystem|SetFloatParameter (Variables|Default|GetNS_AI_Wildlife_InsectSwarm) "Flower" (select (Variables|InternalData|GetIsOnPlant) 1.0 0.0))
          (Class|InsectSwarmSoundLibraryComponent|UpdatePlantStatus (Variables|Default|GetInsectSwarm_SoundLibraryComponent) (Variables|InternalData|GetIsOnPlant)))))
    (:then_1
      (if (not (Variables|InternalData|GetIsOnPlant))
        (CallFunction|ResetCollisionSwarmtoPlant)))))

(fn UpdateSoundLibraryAudio
  (params (InEventName :type Gameplay%20Tag%20Structure) (Play :type Boolean))
  (bind _insectswarm_soundlibrarycomponent (Variables|Default|GetInsectSwarm_SoundLibraryComponent))
  (if Play
    (SoundLibrary|PlayEvent _insectswarm_soundlibrarycomponent InEventName)
    (else
      (SoundLibrary|StopEvent _insectswarm_soundlibrarycomponent InEventName))))

(fn ConstructionScript
  (|CallParentFunction))
